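# UI material for icons in the Creative property editor's icon picker, drawn from distance-field textures.
# It samples a Texture parameter, defaulting to an icon library texture, through a median-style Min/Max network.
# The network reads red twice alongside blue, so as wired it reduces to the red channel.
# MF_UI_AntiAliasing takes that distance value with a mid-point threshold to produce a smooth edge.
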Material /FortniteCoreUI/InGame/Creative/PropertyEditor/Icon_Picker/Materials_DF/M_UI_IconPicker_Icon_DF
graph {
 properties {
  bUsedWithStaticMesh=false
  blendMode=BLEND_Translucent
  materialDomain=MD_UI
  pixelDepthOffsetMode=PDOM_Legacy
  usedShadingModels=Default Lit
 }
 nodes [
  n1 = TextureSampleParameter2D() {
   parameterName=Texture
   samplerType=SAMPLERTYPE_LinearColor
   texture=/FortniteCoreUI/InGame/Creative/PropertyEditor/Icon_Picker/Textures_DF/T_UI_IconLibrary_0_DF
  }
  n2 = Max(A=@n1.R, B=@n1.R)
  MaterialFunctionCall(DistanceField=Max(A=Min(A=@n1.R, B=@n1.R), B=Min(A=@n2, B=@n1.B)), Threshold=Constant() {r=0.5}) {materialFunction=&MF_UI_AntiAliasing}
 ]
}
